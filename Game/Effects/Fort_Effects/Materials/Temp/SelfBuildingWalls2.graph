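# Material function for self-building walls that decodes per-board animation data baked into mesh vertex color and UV channels.
# Vertex color yields world-space board orientation and its up cross vector, while UV2 plus UV3.x yields pivots offset from actor position.
# UV1.x, rounded up, gives default animation order, combined with distance-to-pivot falloff for a secondary gradient.
# UV3.y supplies per-mesh random and ±1 direction; board count passes through as an input.
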MaterialFunction /Game/Effects/Fort_Effects/Materials/Temp/SelfBuildingWalls2
graph {
 properties {description=Self building walls, libraryCategoriesText [Misc,FortNite]}
 nodes [
  n1 = VertexColor()
  n2 = Transform(None=Normalize(VectorInput=ConstantBiasScale(None=@n1) {bias=-0.5, scale=2})) {transformSourceType=TRANSFORMSOURCE_Local}
  FunctionOutput(None=@n2) {outputName=Board Orientation}
  n3 = AppendVector(A=TextureCoordinate() {coordinateIndex=2}, B=ComponentMask(None=TextureCoordinate() {coordinateIndex=3}) {r=true})
  n4 = ComponentMask(None=Add(A=ActorPositionWS(), B=Transform(None=@n3) {transformSourceType=TRANSFORMSOURCE_Local})) {b=true, g=true, r=true}
  FunctionOutput(None=@n4) {outputName=Board Pivot Pos}
  n5 = FunctionInput(Preview=ComponentMask(None=TextureCoordinate() {coordinateIndex=1}) {desc=manual animation, r=true}) {bUsePreviewValueAsDefault=true, inputName=Manual Anim Grad, inputType=FunctionInput_Scalar, sortPriority=9, previewValue=<512 0 0 1>}
  n6 = ComponentMask(None=Ceil(None=@n5) {desc=rounded up from .5}) {desc=manual animation, r=true}
  FunctionOutput(None=@n6) {outputName=Anim Order}
  n7 = FunctionInput() {
   bUsePreviewValueAsDefault=true
   inputName=Sec mot Falloff rad
   inputType=FunctionInput_Scalar
   sortPriority=9
   previewValue=<-27755.5781 0 0 1>
  }
  n8 = Divide(A=Distance(A=WorldPosition().XYZ, B=@n4), B=@n7)
  FunctionOutput(None=Abs(None=Subtract(A=@n6, B=@n8))) {outputName=Anim grad w falloff, sortPriority=13}
  FunctionOutput(None=@n8) {outputName=falloff, sortPriority=14}
  n9 = Constant3Vector() {constant=<rgba 0 0 1 0>}
  FunctionOutput(None=Normalize(VectorInput=CrossProduct(A=@n2, B=@n9))) {outputName=up Cross Orientation}
  n10 = ComponentMask(None=TextureCoordinate() {coordinateIndex=3}) {g=true}
  n11 = Constant() {r=1}
  FunctionOutput(None=If(A=OneMinus(None=@n10), B=@n11, "A > B"=@n11, "A < B"=Constant() {r=-1}) {desc=face boards are white}) {outputName=Direction 1 or -1}
  FunctionOutput(None=ComponentMask(None=TextureCoordinate() {coordinateIndex=3}) {desc=Random, g=true}) {outputName=Random Per Mesh}
  n12 = FunctionInput() {bUsePreviewValueAsDefault=true, inputName=Number of sub objects, inputType=FunctionInput_Scalar, previewValue=<57 0 0 1>}
  FunctionOutput(None=@n12) {outputName=Number of Boards}
  n13 = TextureCoordinate() {coordinateIndex=3}
  n14 = Constant() {r=1}
  If(A=OneMinus(None=ComponentMask(None=@n13) {g=true}), B=@n14, "A > B"=@n14, "A < B"=Constant() {r=-1}) {desc=face boards are white}
 ]
}
